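# Physics asset for the medium female IceRetreat body, in its BackpackOrCape variant, driving secondary motion of the jacket, belt and flaps.
# Capsules on pelvis, thighs, clavicles and upper arms accompany three-bone jacket chains at left and right front, middle and back, plus flaps and a thigh-anchored belt chain.
# Front and middle jacket limits widen toward the tips, while back chains are angularly locked.
# Free-rotation constraints tie the two back tips together and the right middle tip to the right front.
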PhysicsAsset /Game/Characters/Player/Female/Medium/Bodies/F_MED_IceRetreat/Meshes/F_MED_IceRetreat_Physics_BackpackOrCape
bodies {
 base_belt [{shapeType=Capsule, length=0.5, radius=0.5, shapeName=base_belt_capsule}]
 base_jacket_l_back [{shapeType=Capsule, length=0.5, radius=0.5, shapeName=base_jacket_l_back_capsule}]
 base_jacket_l_front [{shapeType=Capsule, length=0.5, radius=0.5, shapeName=base_jacket_l_front_capsule}]
 base_jacket_l_middl [{shapeType=Capsule, length=0.5, radius=0.5, shapeName=base_jacket_l_middl_capsule}]
 base_jacket_r_back [{shapeType=Capsule, length=0.5, radius=0.5, shapeName=base_jacket_r_back_capsule}]
 base_jacket_r_front [{shapeType=Capsule, length=0.5, radius=0.5, shapeName=base_jacket_r_front_capsule}]
 base_jacket_r_middl [{shapeType=Capsule, length=0.5, radius=0.5, shapeName=base_jacket_r_middl_capsule}]
 base_l_flap [{shapeType=Capsule, length=0.5, radius=0.5, shapeName=base_l_flap_capsule}]
 base_r_flap [{shapeType=Capsule, length=0.5, radius=0.5, shapeName=base_r_flap_capsule}]
 clavicle_l [{shapeType=Capsule, length=11, radius=11, shapeName=clavicle_l_capsule, center=<-0.5 0.5 -8>, rotation=<rot 90 0 0>}]
 clavicle_r [{shapeType=Capsule, length=11, radius=11, shapeName=clavicle_r_capsule, center=<-0.5 0.5 8>, rotation=<rot 90 0 0>}]
 dyn_belt_01 [{shapeType=Capsule, length=1, radius=0.8, shapeName=dyn_belt_01_capsule, center=<1 0 0>, rotation=<rot 90 0 0>}]
 dyn_belt_02 [{shapeType=Capsule, length=4, radius=1.3, shapeName=dyn_belt_02_capsule, center=<3.3 -0.3 0>, rotation=<rot 90 0 0>}]
 dyn_belt_03 [{shapeType=Capsule, length=6, radius=1.5, shapeName=dyn_belt_03_capsule, center=<4 -0.3 0>, rotation=<rot 90 0 0>}]
 dyn_belt_04 [{shapeType=Capsule, length=11, radius=1.1, shapeName=dyn_belt_04_capsule, center=<6 -0.2 0>, rotation=<rot 90 0 0>}]
 dyn_jacket_l_back_01 [{shapeType=Capsule, length=10, radius=2, shapeName=dyn_jacket_l_back_01_capsule, center=<5 -1.75 1.9392>, rotation=<rot 90 0 0>}]
 dyn_jacket_l_back_02 [{shapeType=Capsule, length=11, radius=6, shapeName=dyn_jacket_l_back_03_capsule, center=<5.5004 -2.5 0>}]
 dyn_jacket_l_back_03 [{shapeType=Capsule, length=11, radius=6, shapeName=dyn_jacket_l_back_03_capsule, center=<1 -3.5 0>}]
 dyn_jacket_l_front_01 [{shapeType=Capsule, length=12, radius=6, shapeName=dyn_jacket_l_front_02_capsule, center=<6.5 -4 -4>, rotation=<rot 10 0 0>}]
 dyn_jacket_l_front_02 [{shapeType=Capsule, length=19, radius=6, shapeName=dyn_jacket_l_front_02_capsule, center=<6.508 -4.3 -4.0453>, rotation=<rot 10 0 -5>}]
 dyn_jacket_l_front_03 [{shapeType=Capsule, length=16, radius=4, shapeName=dyn_jacket_l_front_03_capsule, center=<1.1805 -2.5 -2>, rotation=<rot 10 20 0>}]
 dyn_jacket_l_middl_01 [{shapeType=Capsule, length=12, radius=6, shapeName=dyn_jacket_l_middl_02_capsule, center=<7.5071 -4 1.1776>, rotation=<rot 10 0 0>}]
 dyn_jacket_l_middl_02 [{shapeType=Capsule, length=12, radius=6, shapeName=dyn_jacket_l_middl_02_capsule, center=<6.5 -4.5 1>}]
 dyn_jacket_l_middl_03 [{shapeType=Capsule, length=22, radius=8, shapeName=dyn_jacket_l_middl_03_capsule, center=<2 -2.5 0>}]
 dyn_jacket_r_back_01 [{shapeType=Capsule, length=10, radius=2, shapeName=dyn_jacket_r_back_01_capsule, center=<5 -1.75 1.8976>, rotation=<rot 90 0 0>}]
 dyn_jacket_r_back_02 [{shapeType=Capsule, length=11, radius=6, shapeName=dyn_jacket_l_back_03_capsule, center=<6.5 -5 0>}]
 dyn_jacket_r_back_03 [{shapeType=Capsule, length=11, radius=6, shapeName=dyn_jacket_l_back_03_capsule, center=<1 -3.5 0>}]
 dyn_jacket_r_front_01 [{shapeType=Capsule, length=12, radius=6, shapeName=dyn_jacket_r_middl_02_capsule, center=<5 -3 4>, rotation=<rot -20 0 15>}]
 dyn_jacket_r_front_02 [{shapeType=Capsule, length=12, radius=6, shapeName=dyn_jacket_r_middl_02_capsule, center=<6.5 -3 4>, rotation=<rot -20 0 15>}]
 dyn_jacket_r_front_03 [{shapeType=Capsule, length=16, radius=6, shapeName=dyn_jacket_l_front_03_capsule, center=<2 -3 3>, rotation=<rot -20 0 15>}]
 dyn_jacket_r_middl_01 [{shapeType=Capsule, length=12, radius=6, shapeName=dyn_jacket_r_middl_02_capsule, center=<5 -3 4>}]
 dyn_jacket_r_middl_02 [{shapeType=Capsule, length=22, radius=6, shapeName=dyn_jacket_r_middl_02_capsule, center=<6.5 -4 4>}]
 dyn_jacket_r_middl_03 [{shapeType=Capsule, length=22, radius=6, shapeName=dyn_jacket_l_front_03_capsule, center=<2 -1.3 3>}]
 dyn_l_flap [{shapeType=Capsule, length=2, radius=1.5, shapeName=dyn_l_flap_capsule, center=<1 0 0>, rotation=<rot 90 0 0>}]
 dyn_r_flap [{shapeType=Capsule, length=2, radius=1.5, shapeName=dyn_r_flap_capsule, center=<1 0 0>, rotation=<rot 90 0 0>}]
 pelvis [{shapeType=Capsule, length=22, radius=11, shapeName=pelvis_capsule, center=<0.8843 -1.7719 -0.0008>, rotation=<rot 90 0 0>}]
 thigh_l [{shapeType=Capsule, length=33, radius=8.3, shapeName=thigh_r_capsule, center=<-16 -0.4 -0.5>, rotation=<rot 85 180 -175>}]
 thigh_r [{shapeType=Capsule, length=22, radius=9.5, shapeName=thigh_r_capsule, center=<18 0 0>, rotation=<rot 90 89.4328 89.4328>}]
 upperarm_l [{shapeType=Capsule, length=11, radius=7, shapeName=upperarm_l_capsule, center=<12 0 0>, rotation=<rot 90 0 0>}]
 upperarm_r [{shapeType=Capsule, length=11, radius=7, shapeName=upperarm_r_capsule, center=<-12 0 0>, rotation=<rot 90 0 0>}]
}
constraints [
 {
  bone1Name=dyn_jacket_l_back_01
  bone2Name=base_jacket_l_back
  swing1LimitDegrees=5
  swing1Motion=Locked
  swing2LimitDegrees=5
  swing2Motion=Locked
  twistLimitDegrees=1
  twistMotion=Locked
 }
 {
  bone1Name=dyn_jacket_l_back_02
  bone2Name=dyn_jacket_l_back_01
  swing1LimitDegrees=2
  swing1Motion=Locked
  swing2LimitDegrees=2
  swing2Motion=Locked
  twistLimitDegrees=2
  twistMotion=Locked
 }
 {
  bone1Name=dyn_jacket_l_back_03
  bone2Name=dyn_jacket_l_back_02
  swing1LimitDegrees=3
  swing1Motion=Locked
  swing2LimitDegrees=5
  swing2Motion=Locked
  twistLimitDegrees=3
  twistMotion=Locked
 }
 {bone1Name=dyn_jacket_l_front_01, bone2Name=base_jacket_l_front, swing1LimitDegrees=25, swing2LimitDegrees=5, twistLimitDegrees=1}
 {bone1Name=dyn_jacket_l_front_02, bone2Name=dyn_jacket_l_front_01, swing1LimitDegrees=35, swing2LimitDegrees=10, twistLimitDegrees=3}
 {bone1Name=dyn_jacket_l_front_03, bone2Name=dyn_jacket_l_front_02, swing1LimitDegrees=45, swing2LimitDegrees=15, twistLimitDegrees=5}
 {bone1Name=dyn_jacket_l_middl_01, bone2Name=base_jacket_l_middl, swing1LimitDegrees=15, swing2LimitDegrees=5, twistLimitDegrees=1}
 {bone1Name=dyn_jacket_l_middl_02, bone2Name=dyn_jacket_l_middl_01, swing1LimitDegrees=20, swing2LimitDegrees=10, twistLimitDegrees=3}
 {bone1Name=dyn_jacket_l_middl_03, bone2Name=dyn_jacket_l_middl_02, swing1LimitDegrees=25, swing2LimitDegrees=15, twistLimitDegrees=5}
 {
  bone1Name=dyn_jacket_r_back_01
  bone2Name=base_jacket_r_back
  swing1LimitDegrees=5
  swing1Motion=Locked
  swing2LimitDegrees=5
  swing2Motion=Locked
  twistLimitDegrees=1
  twistMotion=Locked
 }
 {
  bone1Name=dyn_jacket_r_back_02
  bone2Name=dyn_jacket_r_back_01
  swing1LimitDegrees=2
  swing1Motion=Locked
  swing2LimitDegrees=2
  swing2Motion=Locked
  twistLimitDegrees=2
  twistMotion=Locked
 }
 {
  bone1Name=dyn_jacket_r_back_03
  bone2Name=dyn_jacket_r_back_02
  swing1LimitDegrees=3
  swing1Motion=Locked
  swing2LimitDegrees=5
  swing2Motion=Locked
  twistLimitDegrees=3
  twistMotion=Locked
 }
 {bone1Name=dyn_jacket_r_middl_01, bone2Name=base_jacket_r_middl, swing1LimitDegrees=15, swing2LimitDegrees=5, twistLimitDegrees=1}
 {bone1Name=dyn_jacket_r_middl_02, bone2Name=dyn_jacket_r_middl_01, swing1LimitDegrees=20, swing2LimitDegrees=5, twistLimitDegrees=3}
 {bone1Name=dyn_jacket_r_middl_03, bone2Name=dyn_jacket_r_middl_02, swing1LimitDegrees=25, swing2LimitDegrees=10, twistLimitDegrees=5}
 {
  bone1Name=base_jacket_r_middl
  bone2Name=pelvis
  swing1LimitDegrees=45
  swing1Motion=Free
  swing2LimitDegrees=45
  swing2Motion=Free
  twistLimitDegrees=45
  twistMotion=Free
 }
 {
  bone1Name=base_jacket_r_back
  bone2Name=pelvis
  swing1LimitDegrees=45
  swing1Motion=Free
  swing2LimitDegrees=45
  swing2Motion=Free
  twistLimitDegrees=45
  twistMotion=Free
 }
 {
  bone1Name=base_jacket_l_back
  bone2Name=pelvis
  swing1LimitDegrees=45
  swing1Motion=Free
  swing2LimitDegrees=45
  swing2Motion=Free
  twistLimitDegrees=45
  twistMotion=Free
 }
 {
  bone1Name=base_jacket_l_middl
  bone2Name=pelvis
  swing1LimitDegrees=45
  swing1Motion=Free
  swing2LimitDegrees=45
  swing2Motion=Free
  twistLimitDegrees=45
  twistMotion=Free
 }
 {
  bone1Name=base_jacket_l_front
  bone2Name=pelvis
  swing1LimitDegrees=45
  swing1Motion=Free
  swing2LimitDegrees=45
  swing2Motion=Free
  twistLimitDegrees=45
  twistMotion=Free
 }
 {
  bone1Name=thigh_r
  bone2Name=pelvis
  swing1LimitDegrees=45
  swing1Motion=Free
  swing2LimitDegrees=45
  swing2Motion=Free
  twistLimitDegrees=45
  twistMotion=Free
 }
 {
  bone1Name=base_l_flap
  bone2Name=pelvis
  swing1LimitDegrees=45
  swing1Motion=Free
  swing2LimitDegrees=45
  swing2Motion=Free
  twistLimitDegrees=45
  twistMotion=Free
 }
 {bone1Name=dyn_l_flap, bone2Name=base_l_flap, swing1LimitDegrees=1, swing2LimitDegrees=1, twistLimitDegrees=1}
 {
  bone1Name=base_r_flap
  bone2Name=pelvis
  swing1LimitDegrees=45
  swing1Motion=Free
  swing2LimitDegrees=45
  swing2Motion=Free
  twistLimitDegrees=45
  twistMotion=Free
 }
 {bone1Name=dyn_r_flap, bone2Name=base_r_flap, swing1LimitDegrees=45, swing2LimitDegrees=10, twistLimitDegrees=10}
 {
  bone1Name=base_belt
  bone2Name=thigh_l
  swing1LimitDegrees=45
  swing1Motion=Locked
  swing2LimitDegrees=45
  swing2Motion=Locked
  twistLimitDegrees=45
  twistMotion=Locked
 }
 {bone1Name=dyn_belt_01, bone2Name=base_belt, swing1LimitDegrees=5, swing2LimitDegrees=45, twistLimitDegrees=45}
 {bone1Name=dyn_belt_02, bone2Name=dyn_belt_01, swing1LimitDegrees=25, swing2LimitDegrees=20, twistLimitDegrees=30}
 {bone1Name=dyn_belt_03, bone2Name=dyn_belt_02, swing1LimitDegrees=40, swing2LimitDegrees=25, twistLimitDegrees=20}
 {bone1Name=dyn_belt_04, bone2Name=dyn_belt_03, swing1LimitDegrees=40, swing2LimitDegrees=20, twistLimitDegrees=15}
 {bone1Name=dyn_jacket_r_front_03, bone2Name=dyn_jacket_r_front_02, swing1LimitDegrees=45, swing2LimitDegrees=15, twistLimitDegrees=5}
 {
  bone1Name=base_jacket_r_front
  bone2Name=pelvis
  swing1LimitDegrees=45
  swing1Motion=Free
  swing2LimitDegrees=45
  swing2Motion=Free
  twistLimitDegrees=45
  twistMotion=Free
 }
 {bone1Name=dyn_jacket_r_front_01, bone2Name=base_jacket_r_front, swing1LimitDegrees=25, swing2LimitDegrees=5, twistLimitDegrees=1}
 {bone1Name=dyn_jacket_r_front_02, bone2Name=dyn_jacket_r_front_01, swing1LimitDegrees=35, swing2LimitDegrees=10, twistLimitDegrees=3}
 {
  bone1Name=clavicle_l
  bone2Name=pelvis
  swing1LimitDegrees=45
  swing1Motion=Free
  swing2LimitDegrees=45
  swing2Motion=Free
  twistLimitDegrees=45
  twistMotion=Free
 }
 {
  bone1Name=upperarm_l
  bone2Name=clavicle_l
  swing1LimitDegrees=45
  swing1Motion=Free
  swing2LimitDegrees=45
  swing2Motion=Free
  twistLimitDegrees=45
  twistMotion=Free
 }
 {
  bone1Name=clavicle_r
  bone2Name=pelvis
  swing1LimitDegrees=45
  swing1Motion=Free
  swing2LimitDegrees=45
  swing2Motion=Free
  twistLimitDegrees=45
  twistMotion=Free
 }
 {
  bone1Name=upperarm_r
  bone2Name=clavicle_r
  swing1LimitDegrees=45
  swing1Motion=Free
  swing2LimitDegrees=45
  swing2Motion=Free
  twistLimitDegrees=45
  twistMotion=Free
 }
 {
  bone1Name=dyn_jacket_r_middl_03
  bone2Name=dyn_jacket_r_front_03
  swing1LimitDegrees=45
  swing1Motion=Free
  swing2LimitDegrees=45
  swing2Motion=Free
  twistLimitDegrees=45
  twistMotion=Free
 }
 {
  bone1Name=dyn_jacket_r_back_03
  bone2Name=dyn_jacket_l_back_03
  swing1LimitDegrees=45
  swing1Motion=Free
  swing2LimitDegrees=45
  swing2Motion=Free
  twistLimitDegrees=45
  twistMotion=Free
 }
]
