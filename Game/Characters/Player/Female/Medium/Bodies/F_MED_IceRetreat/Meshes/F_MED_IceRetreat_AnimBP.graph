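# Anim Blueprint for the IceRetreat female medium player body, animating jacket, belt, zippers and jewelry atop the parent mesh's copied pose.
# Control Rig and weighted constraints feed additive bone offsets driven by movement-state alphas (jump, crouch, slide, DBNO).
# Cached branches run RigidBody physics, switched by backpack/cape and in-air state, then apply parachute offsets and AnimDynamics for accessories.
# bRigidBodyEnabled selects between that result and the Cache_RB_Disable pose.
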AnimBlueprint /Game/Characters/Player/Female/Medium/Bodies/F_MED_IceRetreat/Meshes/F_MED_IceRetreat_AnimBP
target_skeleton=&/F_MED_IceRetreat_Skeleton
graphs [
 AnimGraph {nodes [
   n1 = Get isParachuting() -> (isParachuting:double)
   n2 = Use cached pose 'Cache_Wearing_Back'()
   n3 = Use cached pose 'Cache_Unwearing_Back'()
   n4 = Get Is_Wearing_Backpack_or_Cape() -> (Is_Wearing_Backpack_or_Cape:bool)
   n5 = "Transform (Modify) Bone - Bone: base_jacket_l_middl"(
    ComponentPose=Local To Component(LocalPose=Blend Poses by bool(BlendPose_0=@n2, BlendPose_1=@n3, bActiveValue=@n4) {blendType=Linear})
    Alpha=@n1
    Translation="0,0,0"
    Rotation="0,-10,0"
    Scale="1,1,1"
   ) {
    rotationMode=BMM_Additive
    rotationSpace=BCS_BoneSpace
    translationMode=BMM_Additive
    translationSpace=BCS_BoneSpace
    alphaBoolBlend {blendOption=Linear}
    alphaScaleBiasClamp {bInterpResult=true}
    boneToModify {boneName=base_jacket_l_middl}
    rotation {yaw=-10}
   }
   n6 = "Transform (Modify) Bone - Bone: base_jacket_l_back"(ComponentPose=@n5, Alpha=@n1, Translation="0,0,0", Rotation="0,-10,0", Scale="1,1,1") {
    rotationMode=BMM_Additive
    rotationSpace=BCS_BoneSpace
    translationMode=BMM_Additive
    translationSpace=BCS_BoneSpace
    alphaBoolBlend {blendOption=Linear}
    alphaScaleBiasClamp {bInterpResult=true}
    boneToModify {boneName=base_jacket_l_back}
    rotation {yaw=-10}
   }
   n7 = "Transform (Modify) Bone - Bone: base_jacket_r_middl"(ComponentPose=@n6, Alpha=@n1, Translation="0,0,0", Rotation="0,-10,0", Scale="1,1,1") {
    rotationMode=BMM_Additive
    rotationSpace=BCS_BoneSpace
    translationMode=BMM_Additive
    translationSpace=BCS_BoneSpace
    alphaBoolBlend {blendOption=Linear}
    alphaScaleBiasClamp {bInterpResult=true}
    boneToModify {boneName=base_jacket_r_middl}
    rotation {yaw=-10}
   }
   n8 = "Transform (Modify) Bone - Bone: base_jacket_r_back"(ComponentPose=@n7, Alpha=@n1, Translation="0,0,0", Rotation="0,-10,0", Scale="1,1,1") {
    rotationMode=BMM_Additive
    rotationSpace=BCS_BoneSpace
    translationMode=BMM_Additive
    translationSpace=BCS_BoneSpace
    alphaBoolBlend {blendOption=Linear}
    alphaScaleBiasClamp {bInterpResult=true}
    boneToModify {boneName=base_jacket_r_back}
    rotation {yaw=-10}
   }
   n9 = Anim Dynamics - Bone: dyn_zipper_pants(ComponentPose=@n8, Alpha=1, ExternalForce="0,0,0") {
    angularDampingOverride=1
    angularSpringConstant=20000
    bAngularSpring=true
    bOverrideAngularDamping=true
    bUseGravityOverride=true
    gravityScale=0.35
    simulationSpace=RootRelative
    alphaBoolBlend {blendOption=Linear}
    boundBone {boneName=dyn_zipper_pants}
    componentLinearAccScale=<1 1 1>
    componentLinearVelScale=<1 1 1>
    gravityOverride {z=-980}
    physicsBodyDefinitions [{
      boundBone {boneName=dyn_zipper_pants}
      boxExtents=<20 30 30>
      constraintSetup {angularLimitsMax=<5 5 5>, angularLimitsMin=<-5 -5 -80>, angularTarget=<1 0 0>}
      localJointOffset=<0 0 0>
     }]
   }
   n10 = Anim Dynamics - Bone: dyn_zipper_pants_r_01(ComponentPose=@n9, Alpha=1, ExternalForce="0,0,0") {
    angularDampingOverride=1
    angularSpringConstant=20000
    bAngularSpring=true
    bOverrideAngularDamping=true
    bUseGravityOverride=true
    gravityScale=0.35
    simulationSpace=RootRelative
    alphaBoolBlend {blendOption=Linear}
    boundBone {boneName=dyn_zipper_pants_r_01}
    componentLinearAccScale=<1 1 1>
    componentLinearVelScale=<1 1 1>
    gravityOverride {z=-980}
    physicsBodyDefinitions [{
      boundBone {boneName=dyn_zipper_pants_r_01}
      boxExtents=<20 30 30>
      constraintSetup {angularLimitsMax=<5 5 5>, angularLimitsMin=<-5 -5 -80>, angularTarget=<1 0 0>}
      localJointOffset=<0 0 0>
     }]
   }
   n11 = Anim Dynamics - Bone: dyn_zipper_pants_r_02(ComponentPose=@n10, Alpha=1, ExternalForce="0,0,0") {
    angularDampingOverride=1
    angularSpringConstant=20000
    bAngularSpring=true
    bOverrideAngularDamping=true
    bUseGravityOverride=true
    gravityScale=0.35
    simulationSpace=RootRelative
    alphaBoolBlend {blendOption=Linear}
    boundBone {boneName=dyn_zipper_pants_r_02}
    componentLinearAccScale=<1 1 1>
    componentLinearVelScale=<1 1 1>
    gravityOverride {z=-980}
    physicsBodyDefinitions [{
      boundBone {boneName=dyn_zipper_pants_r_02}
      boxExtents=<20 30 30>
      constraintSetup {angularLimitsMax=<5 5 5>, angularLimitsMin=<-5 -5 -80>, angularTarget=<1 0 0>}
      localJointOffset=<0 0 0>
     }]
   }
   n12 = Anim Dynamics - Bone: dyn_zipper_pants_r_03(ComponentPose=@n11, Alpha=1, ExternalForce="0,0,0") {
    angularDampingOverride=1
    angularSpringConstant=20000
    bAngularSpring=true
    bOverrideAngularDamping=true
    bUseGravityOverride=true
    gravityScale=0.35
    simulationSpace=RootRelative
    alphaBoolBlend {blendOption=Linear}
    boundBone {boneName=dyn_zipper_pants_r_03}
    componentLinearAccScale=<1 1 1>
    componentLinearVelScale=<1 1 1>
    gravityOverride {z=-980}
    physicsBodyDefinitions [{
      boundBone {boneName=dyn_zipper_pants_r_03}
      boxExtents=<20 30 30>
      constraintSetup {angularLimitsMax=<5 5 5>, angularLimitsMin=<-5 -5 -80>, angularTarget=<1 0 0>}
      localJointOffset=<0 0 0>
     }]
   }
   n13 = Anim Dynamics - Bone: dyn_necklace_01(ComponentPose=@n12, Alpha=1, ExternalForce="0,0,0") {
    angularDampingOverride=1
    angularSpringConstant=20000
    bAngularSpring=true
    bOverrideAngularDamping=true
    bUseGravityOverride=true
    gravityScale=0.35
    simulationSpace=RootRelative
    alphaBoolBlend {blendOption=Linear}
    boundBone {boneName=dyn_necklace_01}
    componentLinearAccScale=<1 1 1>
    componentLinearVelScale=<1 1 1>
    gravityOverride {z=-980}
    physicsBodyDefinitions [{
      boundBone {boneName=dyn_necklace_01}
      boxExtents=<20 30 30>
      constraintSetup {angularLimitsMax=<2 5 5>, angularLimitsMin=<-2 -5 -15>, angularTarget=<1 0 0>}
      localJointOffset=<0 0 0>
     }]
   }
   n14 = Anim Dynamics - Bone: dyn_necklace_02(ComponentPose=@n13, Alpha=1, ExternalForce="0,0,0") {
    angularDampingOverride=1
    angularSpringConstant=20000
    bAngularSpring=true
    bOverrideAngularDamping=true
    bUseGravityOverride=true
    gravityScale=0.35
    simulationSpace=RootRelative
    alphaBoolBlend {blendOption=Linear}
    boundBone {boneName=dyn_necklace_02}
    componentLinearAccScale=<1 1 1>
    componentLinearVelScale=<1 1 1>
    gravityOverride {z=-980}
    physicsBodyDefinitions [{
      boundBone {boneName=dyn_necklace_02}
      boxExtents=<20 30 30>
      constraintSetup {angularLimitsMax=<2 5 5>, angularLimitsMin=<-2 -5 -15>, angularTarget=<1 0 0>}
      localJointOffset=<0 0 0>
     }]
   }
   n15 = Anim Dynamics - Bone: dyn_bracelet_r(ComponentPose=@n14, Alpha=1, ExternalForce="0,0,0") {
    angularDampingOverride=1
    angularSpringConstant=20000
    bAngularSpring=true
    bOverrideAngularDamping=true
    bUseGravityOverride=true
    gravityScale=0.35
    simulationSpace=RootRelative
    alphaBoolBlend {blendOption=Linear}
    boundBone {boneName=dyn_bracelet_r}
    componentLinearAccScale=<1 1 1>
    componentLinearVelScale=<1 1 1>
    gravityOverride {z=-980}
    physicsBodyDefinitions [{
      boundBone {boneName=dyn_bracelet_r}
      boxExtents=<20 30 30>
      constraintSetup {angularLimitsMax=<15 15 15>, angularLimitsMin=<-15 -15 -15>, angularTarget=<1 0 0>}
      localJointOffset=<0 0 0>
     }]
   }
   n16 = Anim Dynamics - Bone: dyn_bracelet_l(ComponentPose=@n15, Alpha=1, ExternalForce="0,0,0") {
    angularDampingOverride=1
    angularSpringConstant=20000
    bAngularSpring=true
    bOverrideAngularDamping=true
    bUseGravityOverride=true
    gravityScale=0.35
    simulationSpace=RootRelative
    alphaBoolBlend {blendOption=Linear}
    boundBone {boneName=dyn_bracelet_l}
    componentLinearAccScale=<1 1 1>
    componentLinearVelScale=<1 1 1>
    gravityOverride {z=-980}
    physicsBodyDefinitions [{
      boundBone {boneName=dyn_bracelet_l}
      boxExtents=<20 30 30>
      constraintSetup {angularLimitsMax=<15 15 15>, angularLimitsMin=<-15 -15 -15>, angularTarget=<1 0 0>}
      localJointOffset=<0 0 0>
     }]
   }
   Output Pose(Result=Blend Poses by bool(BlendPose_0=Component To Local(ComponentPose=@n16), BlendPose_1=Use cached pose 'Cache_RB_Disable'(), bActiveValue=Get bRigidBodyEnabled()) {blendType=Linear}) {layerGroup=None}
   n17 = Get Sliding() -> (Sliding:double)
   n18 = Get IsJumping() -> (IsJumping:double)
   n19 = Control Rig(Source=Copy Pose From Mesh() {bUseAttachedParent=true}) {
    alphaBoolBlend {blendOption=Linear}
    controlRigAssetReference {blueprintRigClass=&/F_MED_IceRetreat_ControlRig.F_MED_IceRetreat_ControlRig_C}
   }
   n20 = "Constraint - base_belt (PPPPPPPP)"!(
    ComponentPose=Local To Component(LocalPose=@n19)
    Alpha=1
    ConstraintWeights_0=0.3152
    ConstraintWeights_1:float=0.3118
    ConstraintWeights_2:float=0.3007
    ConstraintWeights_3:float=0.0607
    ConstraintWeights_4:float=0.0052
    ConstraintWeights_5:float=0.0033
    ConstraintWeights_6:float=0.0029
    ConstraintWeights_7:float=0.0003
   ) -> (Pose) {
    alphaBoolBlend {blendOption=Linear}
    boneToModify {boneName=base_belt}
    constraintSetup [
     {targetBone {boneName=deform_groin_l}}
     {targetBone {boneName=thigh_twist_01_l}}
     {targetBone {boneName=pelvis}}
     {targetBone {boneName=thigh_l}}
     {targetBone {boneName=dyn_belt_01}}
     {targetBone {boneName=spine_02}}
     {targetBone {boneName=thigh_twist_01_r}}
     {targetBone {boneName=deform_groin_r}}
    ]
    constraintWeights [1,1,1,1,1,1,1,1]
   }
   n21 = "Constraint - base_zipper_pants (PPPPP)"!(
    ComponentPose=@n20
    Alpha=1
    ConstraintWeights_0=0.6821
    ConstraintWeights_1:float=0.2531
    ConstraintWeights_2:float=0.041
    ConstraintWeights_3:float=0.0212
    ConstraintWeights_4:float=0.0026
   ) -> (Pose) {
    alphaBoolBlend {blendOption=Linear}
    boneToModify {boneName=base_zipper_pants}
    constraintSetup [
     {targetBone {boneName=pelvis}}
     {targetBone {boneName=spine_02}}
     {targetBone {boneName=spine_04}}
     {targetBone {boneName=spine_03}}
     {targetBone {boneName=spine_01}}
    ]
    constraintWeights [1,1,1,1,1]
   }
   n22 = "Constraint - base_zipper_pants_r_01 (PPPP)"!(
    ComponentPose=@n21
    Alpha=1
    ConstraintWeights_0=0.4585
    ConstraintWeights_1:float=0.3828
    ConstraintWeights_2:float=0.1584
    ConstraintWeights_3:float=0.0002
   ) -> (Pose) {
    alphaBoolBlend {blendOption=Linear}
    boneToModify {boneName=base_zipper_pants_r_01}
    constraintSetup [
     {targetBone {boneName=thigh_r}}
     {targetBone {boneName=deform_knee_out_r}}
     {targetBone {boneName=calf_r}}
     {targetBone {boneName=thigh_twist_01_r}}
    ]
    constraintWeights [1,1,1,1]
   }
   n23 = "Constraint - base_zipper_pants_r_02 (PPPP)"!(
    ComponentPose=@n22
    Alpha=1
    ConstraintWeights_0=0.4073
    ConstraintWeights_1:float=0.3527
    ConstraintWeights_2:float=0.2398
    ConstraintWeights_3:float=0.0002
   ) -> (Pose) {
    alphaBoolBlend {blendOption=Linear}
    boneToModify {boneName=base_zipper_pants_r_02}
    constraintSetup [
     {targetBone {boneName=deform_knee_out_r}}
     {targetBone {boneName=calf_r}}
     {targetBone {boneName=thigh_r}}
     {targetBone {boneName=calf_twist_02_r}}
    ]
    constraintWeights [1,1,1,1]
   }
   n24 = "Constraint - base_zipper_pants_r_03 (PPPP)"!(
    ComponentPose=@n23
    Alpha=1
    ConstraintWeights_0=0.7158
    ConstraintWeights_1:float=0.2102
    ConstraintWeights_2:float=0.0522
    ConstraintWeights_3:float=0.0218
   ) -> (Pose) {
    alphaBoolBlend {blendOption=Linear}
    boneToModify {boneName=base_zipper_pants_r_03}
    constraintSetup [
     {targetBone {boneName=calf_r}}
     {targetBone {boneName=deform_knee_out_r}}
     {targetBone {boneName=thigh_r}}
     {targetBone {boneName=calf_twist_02_r}}
    ]
    constraintWeights [1,1,1,1]
   }
   n25 = "Transform (Modify) Bone - Bone: dyn_belt_01"(ComponentPose=@n24, Alpha=Get AimingUp() -> (AimingUp:double), Translation="0,0,0", Rotation="0,45,0", Scale="1,1,1") {
    rotationMode=BMM_Additive
    rotationSpace=BCS_BoneSpace
    scaleMode=BMM_Additive
    scaleSpace=BCS_BoneSpace
    alphaBoolBlend {blendOption=Linear}
    alphaScaleBiasClamp {bInterpResult=true}
    boneToModify {boneName=dyn_belt_01}
    rotation {yaw=45}
   }
   n26 = "Transform (Modify) Bone - Bone: base_jacket_r_front"(ComponentPose=@n25, Alpha=@n18, Translation="0,0,0.5", Rotation="-10,-20,20", Scale="1,1,1") {
    rotationMode=BMM_Additive
    rotationSpace=BCS_BoneSpace
    translationMode=BMM_Additive
    translationSpace=BCS_BoneSpace
    alphaBoolBlend {blendOption=Linear}
    alphaScaleBiasClamp {bInterpResult=true}
    boneToModify {boneName=base_jacket_r_front}
    rotation=<rot -10 -20 20>
    translation {z=0.5}
   }
   n27 = "Transform (Modify) Bone - Bone: base_jacket_r_middl"(ComponentPose=@n26, Alpha=@n18, Translation="0,0,0", Rotation="-10,0,0", Scale="1,1,1") {
    rotationMode=BMM_Additive
    rotationSpace=BCS_BoneSpace
    translationMode=BMM_Additive
    translationSpace=BCS_BoneSpace
    alphaBoolBlend {blendOption=Linear}
    alphaScaleBiasClamp {bInterpResult=true}
    boneToModify {boneName=base_jacket_r_middl}
    rotation {pitch=-10}
   }
   n28 = "Transform (Modify) Bone - Bone: base_jacket_r_back"(ComponentPose=@n27, Alpha=@n18, Translation="1.5,0,-2", Rotation="0,0,0", Scale="1,1,1") {
    rotationMode=BMM_Additive
    rotationSpace=BCS_BoneSpace
    translationMode=BMM_Additive
    translationSpace=BCS_BoneSpace
    alphaBoolBlend {blendOption=Linear}
    alphaScaleBiasClamp {bInterpResult=true}
    boneToModify {boneName=base_jacket_r_back}
    translation {x=1.5, z=-2}
   }
   n29 = "Transform (Modify) Bone - Bone: base_belt"(ComponentPose=@n28, Alpha=@n18, Translation="0,0,0", Rotation="-15,90,70", Scale="1,1,1") {
    rotationMode=BMM_Additive
    rotationSpace=BCS_BoneSpace
    translationMode=BMM_Additive
    translationSpace=BCS_BoneSpace
    alphaBoolBlend {blendOption=Linear}
    alphaScaleBiasClamp {bInterpResult=true}
    boneToModify {boneName=base_belt}
    rotation=<rot -15 90 70>
   }
   n30 = "Transform (Modify) Bone - Bone: dyn_belt_01"(ComponentPose=@n29, Alpha=@n18, Translation="0,0,0", Rotation="0,-60,0", Scale="1,1,1") {
    rotationMode=BMM_Additive
    rotationSpace=BCS_BoneSpace
    scaleMode=BMM_Additive
    scaleSpace=BCS_BoneSpace
    alphaBoolBlend {blendOption=Linear}
    alphaScaleBiasClamp {bInterpResult=true}
    boneToModify {boneName=dyn_belt_01}
    rotation {yaw=-60}
   }
   n31 = "Transform (Modify) Bone - Bone: base_jacket_r_front"(ComponentPose=@n30, Alpha=Get Leg_R_Rotate() -> (Leg_R_Rotate:double), Translation="1,1,1", Rotation="0,-50,50", Scale="1,1,1") {
    rotationMode=BMM_Additive
    rotationSpace=BCS_BoneSpace
    translationMode=BMM_Additive
    translationSpace=BCS_BoneSpace
    alphaBoolBlend {blendOption=Linear}
    alphaScaleBiasClamp {bInterpResult=true}
    boneToModify {boneName=base_jacket_r_front}
    rotation {roll=50, yaw=-50}
    translation=<1 1 1>
   }
   n32 = Get Leg_L_Rotate() -> (Leg_L_Rotate:double)
   n33 = "Transform (Modify) Bone - Bone: base_jacket_l_back"(ComponentPose=@n31, Alpha=@n32, Translation="-1.5,0,0", Rotation="0,0,0", Scale="1,1,1") {
    rotationMode=BMM_Additive
    rotationSpace=BCS_BoneSpace
    translationMode=BMM_Additive
    translationSpace=BCS_BoneSpace
    alphaBoolBlend {blendOption=Linear}
    alphaScaleBiasClamp {bInterpResult=true}
    boneToModify {boneName=base_jacket_l_back}
    translation {x=-1.5}
   }
   n34 = "Transform (Modify) Bone - Bone: base_jacket_l_front"(ComponentPose=@n33, Alpha=@n32, Translation="0,-3,0", Rotation="0,-45,-65", Scale="1,1,1") {
    rotationMode=BMM_Additive
    rotationSpace=BCS_BoneSpace
    translationMode=BMM_Additive
    translationSpace=BCS_BoneSpace
    alphaBoolBlend {blendOption=Linear}
    alphaScaleBiasClamp {bInterpResult=true}
    boneToModify {boneName=base_jacket_l_front}
    rotation {roll=-65, yaw=-45}
    translation {y=-3}
   }
   n35 = "Transform (Modify) Bone - Bone: base_belt"(ComponentPose=@n34, Alpha=@n32, Translation="-0.5,0,-1.5", Rotation="90,20,-18", Scale="1,1,1") {
    rotationMode=BMM_Additive
    rotationSpace=BCS_BoneSpace
    scaleMode=BMM_Additive
    scaleSpace=BCS_BoneSpace
    translationMode=BMM_Additive
    translationSpace=BCS_BoneSpace
    alphaBoolBlend {blendOption=Linear}
    alphaScaleBiasClamp {bInterpResult=true}
    boneToModify {boneName=base_belt}
    rotation=<rot 90 20 -18>
    translation {x=-0.5, z=-1.5}
   }
   n36 = "Transform (Modify) Bone - Bone: base_jacket_l_front"(ComponentPose=@n35, Alpha=Get InCrouching() -> (InCrouching:double), Translation="0,-1,-1", Rotation="0,-6,0", Scale="1,1,1") {
    rotationMode=BMM_Additive
    rotationSpace=BCS_BoneSpace
    scaleMode=BMM_Additive
    scaleSpace=BCS_BoneSpace
    translationMode=BMM_Additive
    translationSpace=BCS_BoneSpace
    alphaBoolBlend {blendOption=Linear}
    alphaScaleBiasClamp {bInterpResult=true}
    boneToModify {boneName=base_jacket_l_front}
    rotation {yaw=-6}
    translation {y=-1, z=-1}
   }
   n37 = "Transform (Modify) Bone - Bone: base_belt"(ComponentPose=@n36, Alpha=@n17, Translation="0,0.5,0", Rotation="0,0,20", Scale="1,1,1") {
    rotationMode=BMM_Additive
    rotationSpace=BCS_BoneSpace
    scaleMode=BMM_Additive
    scaleSpace=BCS_BoneSpace
    translationMode=BMM_Additive
    translationSpace=BCS_BoneSpace
    alphaBoolBlend {blendOption=Linear}
    alphaScaleBiasClamp {bInterpResult=true}
    boneToModify {boneName=base_belt}
    rotation {roll=20}
    translation {y=0.5}
   }
   n38 = "Transform (Modify) Bone - Bone: base_jacket_l_front"(ComponentPose=@n37, Alpha=@n17, Translation="0,-4,-1", Rotation="0,-6,0", Scale="1,1,1") {
    rotationMode=BMM_Additive
    rotationSpace=BCS_BoneSpace
    scaleMode=BMM_Additive
    scaleSpace=BCS_BoneSpace
    translationMode=BMM_Additive
    translationSpace=BCS_BoneSpace
    alphaBoolBlend {blendOption=Linear}
    alphaScaleBiasClamp {bInterpResult=true}
    boneToModify {boneName=base_jacket_l_front}
    rotation {yaw=-6}
    translation {y=-4, z=-1}
   }
   n39 = "Transform (Modify) Bone - Bone: base_jacket_r_front"(ComponentPose=@n38, Alpha=Get InAir() -> (InAir:double), Translation="5,0,-2.5", Rotation="0,0,0", Scale="1,1,1") {
    rotationMode=BMM_Additive
    rotationSpace=BCS_BoneSpace
    translationMode=BMM_Additive
    translationSpace=BCS_BoneSpace
    alphaBoolBlend {blendOption=Linear}
    alphaScaleBiasClamp {bInterpResult=true}
    boneToModify {boneName=base_jacket_r_front}
    translation {x=5, z=-2.5}
   }
   n40 = Get Is DBNO() -> (Is DBNO:double)
   n41 = "Transform (Modify) Bone - Bone: base_belt"(ComponentPose=@n39, Alpha=@n40, Translation="0,0,0", Rotation="0,20,0", Scale="1,1,1") {
    rotationMode=BMM_Additive
    rotationSpace=BCS_BoneSpace
    translationMode=BMM_Additive
    translationSpace=BCS_BoneSpace
    alphaBoolBlend {blendOption=Linear}
    alphaScaleBiasClamp {bInterpResult=true}
    boneToModify {boneName=base_belt}
    rotation {yaw=20}
   }
   n42 = "Transform (Modify) Bone - Bone: base_jacket_r_middl"(ComponentPose=@n41, Alpha=@n40, Translation="0,0,-2", Rotation="0,-15,-20", Scale="1,1,1") {
    rotationMode=BMM_Additive
    rotationSpace=BCS_BoneSpace
    scaleMode=BMM_Additive
    scaleSpace=BCS_BoneSpace
    translationMode=BMM_Additive
    translationSpace=BCS_BoneSpace
    alphaBoolBlend {blendOption=Linear}
    alphaScaleBiasClamp {bInterpResult=true}
    boneToModify {boneName=base_jacket_r_middl}
    rotation {roll=-20, yaw=-15}
    translation {z=-2}
   }
   n43 = "Transform (Modify) Bone - Bone: base_jacket_r_front"(ComponentPose=@n42, Alpha=@n40, Translation="4,-4,-2", Rotation="-10,-30,-40", Scale="1,1,1") {
    rotationMode=BMM_Additive
    rotationSpace=BCS_BoneSpace
    scaleMode=BMM_Additive
    scaleSpace=BCS_BoneSpace
    translationMode=BMM_Additive
    translationSpace=BCS_BoneSpace
    alphaBoolBlend {blendOption=Linear}
    alphaScaleBiasClamp {bInterpResult=true}
    boneToModify {boneName=base_jacket_r_front}
    rotation=<rot -10 -30 -40>
    translation=<4 -4 -2>
   }
   n44 = "Transform (Modify) Bone - Bone: base_jacket_r_front"(ComponentPose=@n43, Alpha=Get On Rail() -> (On Rail:double), Translation="0,0,2", Rotation="0,0,55", Scale="1,1,1") {
    rotationMode=BMM_Additive
    rotationSpace=BCS_BoneSpace
    translationMode=BMM_Additive
    translationSpace=BCS_BoneSpace
    alphaBoolBlend {blendOption=Linear}
    alphaScaleBiasClamp {bInterpResult=true}
    boneToModify {boneName=base_jacket_r_front}
    rotation {roll=55}
    translation {z=2}
   }
   n45 = "Transform (Modify) Bone - Bone: base_jacket_l_front"(ComponentPose=@n44, Alpha=Get Head Front() -> (Head Front:double), Translation="5,-3,0", Rotation="0,0,60", Scale="1,1,1") {
    rotationMode=BMM_Additive
    rotationSpace=BCS_BoneSpace
    translationMode=BMM_Additive
    translationSpace=BCS_BoneSpace
    alphaBoolBlend {blendOption=Linear}
    alphaScaleBiasClamp {bInterpResult=true}
    boneToModify {boneName=base_jacket_l_front}
    rotation {roll=60}
    translation=<5 -3>
   }
   Save cached pose 'CACHE_NormalState'(Pose=Component To Local(ComponentPose=@n45)) {cacheName=CACHE_NormalState}
   n46 = Get IsJumping()
   n47 = "Transform (Modify) Bone - Bone: base_l_flap"(
    ComponentPose=Local To Component(LocalPose=Use cached pose 'CACHE_NormalState'())
    Alpha=@n46
    Translation="-1,3,-1"
    Rotation="0,0,0"
    Scale="1,1,1"
   ) {
    rotationMode=BMM_Additive
    rotationSpace=BCS_BoneSpace
    translationMode=BMM_Additive
    translationSpace=BCS_BoneSpace
    alphaBoolBlend {blendOption=Linear}
    alphaScaleBiasClamp {bInterpResult=true}
    boneToModify {boneName=base_l_flap}
    translation=<-1 3 -1>
   }
   n48 = "Transform (Modify) Bone - Bone: base_jacket_l_back"(ComponentPose=@n47, Alpha=@n46, Translation="0,0,0", Rotation="0,30,0", Scale="1,1,1") {
    rotationMode=BMM_Additive
    rotationSpace=BCS_BoneSpace
    translationMode=BMM_Additive
    translationSpace=BCS_BoneSpace
    alphaBoolBlend {blendOption=Linear}
    alphaScaleBiasClamp {bInterpResult=true}
    boneToModify {boneName=base_jacket_l_back}
    rotation {yaw=30}
   }
   n49 = "Transform (Modify) Bone - Bone: base_jacket_r_back"(ComponentPose=@n48, Alpha=@n46, Translation="0,0,0", Rotation="0,30,0", Scale="1,1,1") {
    rotationMode=BMM_Additive
    rotationSpace=BCS_BoneSpace
    translationMode=BMM_Additive
    translationSpace=BCS_BoneSpace
    alphaBoolBlend {blendOption=Linear}
    alphaScaleBiasClamp {bInterpResult=true}
    boneToModify {boneName=base_jacket_r_back}
    rotation {yaw=30}
   }
   n50 = Get InCrouching()
   n51 = "Transform (Modify) Bone - Bone: base_jacket_l_back"(ComponentPose=@n49, Alpha=@n50, Translation="0,0,0", Rotation="0,15,0", Scale="1,1,1") {
    rotationMode=BMM_Additive
    rotationSpace=BCS_BoneSpace
    translationMode=BMM_Additive
    translationSpace=BCS_BoneSpace
    alphaBoolBlend {blendOption=Linear}
    alphaScaleBiasClamp {bInterpResult=true}
    boneToModify {boneName=base_jacket_l_back}
    rotation {yaw=15}
   }
   n52 = "Transform (Modify) Bone - Bone: base_jacket_r_back"(ComponentPose=@n51, Alpha=@n50, Translation="0,0,0", Rotation="0,15,0", Scale="1,1,1") {
    rotationMode=BMM_Additive
    rotationSpace=BCS_BoneSpace
    translationMode=BMM_Additive
    translationSpace=BCS_BoneSpace
    alphaBoolBlend {blendOption=Linear}
    alphaScaleBiasClamp {bInterpResult=true}
    boneToModify {boneName=base_jacket_r_back}
    rotation {yaw=15}
   }
   Save cached pose 'Cache_Wearing_BackpackOrCape'(Pose=Component To Local(ComponentPose=@n52)) {cacheName=Cache_Wearing_BackpackOrCape}
   n53 = Get InAir()
   n54 = "Transform (Modify) Bone - Bone: base_jacket_l_back"(
    ComponentPose=Local To Component(LocalPose=Use cached pose 'CACHE_NormalState'())
    Alpha=@n53
    Translation="0,-0.5,0"
    Rotation="-10,10,0"
    Scale="1,1,1"
   ) {
    rotationMode=BMM_Additive
    rotationSpace=BCS_BoneSpace
    translationMode=BMM_Additive
    translationSpace=BCS_BoneSpace
    alphaBoolBlend {blendOption=Linear}
    alphaScaleBiasClamp {bInterpResult=true}
    boneToModify {boneName=base_jacket_l_back}
    rotation {pitch=-10, yaw=10}
    translation {y=-0.5}
   }
   n55 = "Transform (Modify) Bone - Bone: base_jacket_r_back"(ComponentPose=@n54, Alpha=@n53, Translation="0,-1.5,0", Rotation="13,4,0", Scale="1,1,1") {
    rotationMode=BMM_Additive
    rotationSpace=BCS_BoneSpace
    translationMode=BMM_Additive
    translationSpace=BCS_BoneSpace
    alphaBoolBlend {blendOption=Linear}
    alphaScaleBiasClamp {bInterpResult=true}
    boneToModify {boneName=base_jacket_r_back}
    rotation {pitch=13, yaw=4}
    translation {y=-1.5}
   }
   n56 = RigidBody!(ComponentPose=@n55, Alpha, OverrideWorldGravity:Vector=Get GravityOverride(), ExternalForce) -> (Pose) {
    bClampLinearTranslationLimitToRefPose=true
    bOverrideWorldGravity=true
    overridePhysicsAsset=&/F_MED_IceRetreat_Physics
    simulationSpace=BaseBoneSpace
    alphaBoolBlend {blendOption=Linear}
    baseBoneRef {boneName=spine_05}
    componentAppliedLinearAccClamp=<3000 3000 3000>
    componentLinearAccScale=<0.6 0.6 0.6>
    componentLinearVelScale=<0.6 0.6 0.6>
    simSpaceSettings {
     maxAngularAcceleration=1000
     maxAngularVelocity=3
     maxLinearAcceleration=700
     maxLinearVelocity=500
     velocityScaleZ=0
     worldAlpha=0.2
     externalLinearDragV=<0.2 0.2>
    }
   }
   Save cached pose 'Cache_Unwearing_Back'(Pose=Component To Local(ComponentPose=@n56)) {cacheName=Cache_Unwearing_Back}
   n57 = RigidBody(ComponentPose=Local To Component(LocalPose=Use cached pose 'Cache_Wearing_BackpackOrCape'()), OverrideWorldGravity=Get GravityOverride()) {
    bClampLinearTranslationLimitToRefPose=true
    bOverrideWorldGravity=true
    overridePhysicsAsset=&/F_MED_IceRetreat_Physics_BackpackOrCape
    simulationSpace=BaseBoneSpace
    alphaBoolBlend {blendOption=Linear}
    baseBoneRef {boneName=spine_05}
    componentAppliedLinearAccClamp=<3000 3000 3000>
    componentLinearAccScale=<0.6 0.6 0.6>
    componentLinearVelScale=<0.6 0.6 0.6>
    simSpaceSettings {
     maxAngularAcceleration=1000
     maxAngularVelocity=3
     maxLinearAcceleration=700
     maxLinearVelocity=500
     velocityScaleZ=0
     worldAlpha=0.2
     externalLinearDragV=<0.2 0.2>
    }
   }
   n58 = Get InAir()
   n59 = "Transform (Modify) Bone - Bone: base_jacket_l_back"(
    ComponentPose=Local To Component(LocalPose=Use cached pose 'Cache_Wearing_BackpackOrCape'())
    Alpha=@n58
    Translation="0,-0.5,0"
    Rotation="-10,10,0"
    Scale="1,1,1"
   ) {
    rotationMode=BMM_Additive
    rotationSpace=BCS_BoneSpace
    translationMode=BMM_Additive
    translationSpace=BCS_BoneSpace
    alphaBoolBlend {blendOption=Linear}
    alphaScaleBiasClamp {bInterpResult=true}
    boneToModify {boneName=base_jacket_l_back}
    rotation {pitch=-10, yaw=10}
    translation {y=-0.5}
   }
   n60 = "Transform (Modify) Bone - Bone: base_jacket_r_back"(ComponentPose=@n59, Alpha=@n58, Translation="0,-1.5,0", Rotation="13,4,0", Scale="1,1,1") {
    rotationMode=BMM_Additive
    rotationSpace=BCS_BoneSpace
    translationMode=BMM_Additive
    translationSpace=BCS_BoneSpace
    alphaBoolBlend {blendOption=Linear}
    alphaScaleBiasClamp {bInterpResult=true}
    boneToModify {boneName=base_jacket_r_back}
    rotation {pitch=13, yaw=4}
    translation {y=-1.5}
   }
   n61 = RigidBody(ComponentPose=@n60, OverrideWorldGravity=Get GravityOverride()) {
    bClampLinearTranslationLimitToRefPose=true
    bOverrideWorldGravity=true
    overridePhysicsAsset=&F_MED_IceRetreat_Physics
    simulationSpace=BaseBoneSpace
    alphaBoolBlend {blendOption=Linear}
    baseBoneRef {boneName=spine_05}
    componentAppliedLinearAccClamp=<3000 3000 3000>
    componentLinearAccScale=<0.6 0.6 0.6>
    componentLinearVelScale=<0.6 0.6 0.6>
    simSpaceSettings {
     maxAngularAcceleration=1000
     maxAngularVelocity=3
     maxLinearAcceleration=700
     maxLinearVelocity=500
     velocityScaleZ=0
     worldAlpha=0.2
     externalLinearDragV=<0.2 0.2>
    }
   }
   Save cached pose 'Cache_Wearing_Back'(Pose=Blend Poses by bool(
     BlendPose_0=Component To Local(ComponentPose=@n61)
     BlendPose_1=Component To Local(ComponentPose=@n57)
     bActiveValue=Get InAir_Bool() -> (InAir_Bool:bool)
    ) {blendType=Linear}) {cacheName=Cache_Wearing_Back}
   n62 = Get Sliding()
   n63 = Get Moving2D() -> (Moving2D:double)
   n64 = Get Is DBNO()
   n65 = Get InCrouching()
   n66 = Get InAir()
   n67 = Get IsJumping()
   n68 = "Transform (Modify) Bone - Bone: base_jacket_r_front"(
    ComponentPose=Local To Component(LocalPose=Use cached pose 'CACHE_NormalState'())
    Alpha=@n67
    Translation="0,0,0"
    Rotation="0,-35,0"
    Scale="1,1,1"
   ) {
    rotationMode=BMM_Additive
    rotationSpace=BCS_BoneSpace
    translationMode=BMM_Additive
    translationSpace=BCS_BoneSpace
    alphaBoolBlend {blendOption=Linear}
    alphaScaleBiasClamp {bInterpResult=true}
    boneToModify {boneName=base_jacket_r_front}
    rotation {yaw=-35}
   }
   n69 = "Transform (Modify) Bone - Bone: base_jacket_r_middl"(ComponentPose=@n68, Alpha=@n67, Translation="0,0,0", Rotation="0,-10,0", Scale="1,1,1") {
    rotationMode=BMM_Additive
    rotationSpace=BCS_BoneSpace
    translationMode=BMM_Additive
    translationSpace=BCS_BoneSpace
    alphaBoolBlend {blendOption=Linear}
    alphaScaleBiasClamp {bInterpResult=true}
    boneToModify {boneName=base_jacket_r_middl}
    rotation {yaw=-10}
   }
   n70 = "Transform (Modify) Bone - Bone: base_jacket_l_front"(ComponentPose=@n69, Alpha=@n67, Translation="0,0,0", Rotation="0,-45,0", Scale="1,1,1") {
    rotationMode=BMM_Additive
    rotationSpace=BCS_BoneSpace
    translationMode=BMM_Additive
    translationSpace=BCS_BoneSpace
    alphaBoolBlend {blendOption=Linear}
    alphaScaleBiasClamp {bInterpResult=true}
    boneToModify {boneName=base_jacket_l_front}
    rotation {yaw=-45}
   }
   n71 = "Transform (Modify) Bone - Bone: base_jacket_l_middl"(ComponentPose=@n70, Alpha=@n67, Translation="0,0,0", Rotation="0,-20,0", Scale="1,1,1") {
    rotationMode=BMM_Additive
    rotationSpace=BCS_BoneSpace
    translationMode=BMM_Additive
    translationSpace=BCS_BoneSpace
    alphaBoolBlend {blendOption=Linear}
    alphaScaleBiasClamp {bInterpResult=true}
    boneToModify {boneName=base_jacket_l_middl}
    rotation {yaw=-20}
   }
   n72 = "Transform (Modify) Bone - Bone: base_belt"(ComponentPose=@n71, Alpha=@n67, Translation="0,0,0", Rotation="5,-6,0", Scale="1,1,1") {
    rotationMode=BMM_Additive
    rotationSpace=BCS_BoneSpace
    translationMode=BMM_Additive
    translationSpace=BCS_BoneSpace
    alphaBoolBlend {blendOption=Linear}
    alphaScaleBiasClamp {bInterpResult=true}
    boneToModify {boneName=base_belt}
    rotation {pitch=5, yaw=-6}
   }
   n73 = "Transform (Modify) Bone - Bone: base_jacket_r_back"(ComponentPose=@n72, Alpha=@n66, Translation="0,-4,0", Rotation="0,-20,-20", Scale="1,1,1") {
    rotationMode=BMM_Additive
    rotationSpace=BCS_BoneSpace
    translationMode=BMM_Additive
    translationSpace=BCS_BoneSpace
    alphaBoolBlend {blendOption=Linear}
    alphaScaleBiasClamp {bInterpResult=true}
    boneToModify {boneName=base_jacket_r_back}
    rotation {roll=-20, yaw=-20}
    translation {y=-4}
   }
   n74 = "Transform (Modify) Bone - Bone: base_jacket_r_middl"(ComponentPose=@n73, Alpha=@n66, Translation="0,0,0", Rotation="0,-5,0", Scale="1,1,1") {
    rotationMode=BMM_Additive
    rotationSpace=BCS_BoneSpace
    translationMode=BMM_Additive
    translationSpace=BCS_BoneSpace
    alphaBoolBlend {blendOption=Linear}
    alphaScaleBiasClamp {bInterpResult=true}
    boneToModify {boneName=base_jacket_r_middl}
    rotation {yaw=-5}
   }
   n75 = "Transform (Modify) Bone - Bone: base_jacket_l_back"(ComponentPose=@n74, Alpha=@n66, Translation="0,-4,0", Rotation="0,-15,20", Scale="1,1,1") {
    rotationMode=BMM_Additive
    rotationSpace=BCS_BoneSpace
    translationMode=BMM_Additive
    translationSpace=BCS_BoneSpace
    alphaBoolBlend {blendOption=Linear}
    alphaScaleBiasClamp {bInterpResult=true}
    boneToModify {boneName=base_jacket_l_back}
    rotation {roll=20, yaw=-15}
    translation {y=-4}
   }
   n76 = "Transform (Modify) Bone - Bone: base_jacket_l_middl"(ComponentPose=@n75, Alpha=@n66, Translation="0,0,0", Rotation="0,0,0", Scale="1,1,1") {
    rotationMode=BMM_Additive
    rotationSpace=BCS_BoneSpace
    translationMode=BMM_Additive
    translationSpace=BCS_BoneSpace
    alphaBoolBlend {blendOption=Linear}
    alphaScaleBiasClamp {bInterpResult=true}
    boneToModify {boneName=base_jacket_l_middl}
   }
   n77 = "Transform (Modify) Bone - Bone: base_jacket_r_front"(ComponentPose=@n76, Alpha=@n65, Translation="0,0,0", Rotation="0,-40,0", Scale="1,1,1") {
    rotationMode=BMM_Additive
    rotationSpace=BCS_BoneSpace
    translationMode=BMM_Additive
    translationSpace=BCS_BoneSpace
    alphaBoolBlend {blendOption=Linear}
    alphaScaleBiasClamp {bInterpResult=true}
    boneToModify {boneName=base_jacket_r_front}
    rotation {yaw=-40}
   }
   n78 = "Transform (Modify) Bone - Bone: base_jacket_r_middl"(ComponentPose=@n77, Alpha=@n65, Translation="0,0,0", Rotation="0,-5,0", Scale="1,1,1") {
    rotationMode=BMM_Additive
    rotationSpace=BCS_BoneSpace
    translationMode=BMM_Additive
    translationSpace=BCS_BoneSpace
    alphaBoolBlend {blendOption=Linear}
    alphaScaleBiasClamp {bInterpResult=true}
    boneToModify {boneName=base_jacket_r_middl}
    rotation {yaw=-5}
   }
   n79 = "Transform (Modify) Bone - Bone: base_jacket_l_front"(ComponentPose=@n78, Alpha=@n65, Translation="0,0,0", Rotation="0,-5,0", Scale="1,1,1") {
    rotationMode=BMM_Additive
    rotationSpace=BCS_BoneSpace
    translationMode=BMM_Additive
    translationSpace=BCS_BoneSpace
    alphaBoolBlend {blendOption=Linear}
    alphaScaleBiasClamp {bInterpResult=true}
    boneToModify {boneName=base_jacket_l_front}
    rotation {yaw=-5}
   }
   n80 = "Transform (Modify) Bone - Bone: base_jacket_l_middl"(ComponentPose=@n79, Alpha=@n65, Translation="0,0,0", Rotation="0,-15,0", Scale="1,1,1") {
    rotationMode=BMM_Additive
    rotationSpace=BCS_BoneSpace
    translationMode=BMM_Additive
    translationSpace=BCS_BoneSpace
    alphaBoolBlend {blendOption=Linear}
    alphaScaleBiasClamp {bInterpResult=true}
    boneToModify {boneName=base_jacket_l_middl}
    rotation {yaw=-15}
   }
   n81 = "Transform (Modify) Bone - Bone: base_belt"(ComponentPose=@n80, Alpha=@n65, Translation="0,0,0", Rotation="-10,20,0", Scale="1,1,1") {
    rotationMode=BMM_Additive
    rotationSpace=BCS_BoneSpace
    translationMode=BMM_Additive
    translationSpace=BCS_BoneSpace
    alphaBoolBlend {blendOption=Linear}
    alphaScaleBiasClamp {bInterpResult=true}
    boneToModify {boneName=base_belt}
    rotation {pitch=-10, yaw=20}
   }
   n82 = "Transform (Modify) Bone - Bone: base_jacket_r_front"(ComponentPose=@n81, Alpha=@n64, Translation="0,0,0", Rotation="0,-45,0", Scale="1,1,1") {
    rotationMode=BMM_Additive
    rotationSpace=BCS_BoneSpace
    translationMode=BMM_Additive
    translationSpace=BCS_BoneSpace
    alphaBoolBlend {blendOption=Linear}
    alphaScaleBiasClamp {bInterpResult=true}
    boneToModify {boneName=base_jacket_r_front}
    rotation {yaw=-45}
   }
   n83 = "Transform (Modify) Bone - Bone: base_jacket_r_middl"(ComponentPose=@n82, Alpha=@n64, Translation="0,0,0", Rotation="0,-30,20", Scale="1,1,1") {
    rotationMode=BMM_Additive
    rotationSpace=BCS_BoneSpace
    translationMode=BMM_Additive
    translationSpace=BCS_BoneSpace
    alphaBoolBlend {blendOption=Linear}
    alphaScaleBiasClamp {bInterpResult=true}
    boneToModify {boneName=base_jacket_r_middl}
    rotation {roll=20, yaw=-30}
   }
   n84 = "Transform (Modify) Bone - Bone: base_jacket_l_front"(ComponentPose=@n83, Alpha=@n64, Translation="0,0,0", Rotation="0,-20,-10", Scale="1,1,1") {
    rotationMode=BMM_Additive
    rotationSpace=BCS_BoneSpace
    translationMode=BMM_Additive
    translationSpace=BCS_BoneSpace
    alphaBoolBlend {blendOption=Linear}
    alphaScaleBiasClamp {bInterpResult=true}
    boneToModify {boneName=base_jacket_l_front}
    rotation {roll=-10, yaw=-20}
   }
   n85 = "Transform (Modify) Bone - Bone: base_jacket_l_middl"(ComponentPose=@n84, Alpha=@n64, Translation="0,0,0", Rotation="0,0,0", Scale="1,1,1") {
    rotationMode=BMM_Additive
    rotationSpace=BCS_BoneSpace
    translationMode=BMM_Additive
    translationSpace=BCS_BoneSpace
    alphaBoolBlend {blendOption=Linear}
    alphaScaleBiasClamp {bInterpResult=true}
    boneToModify {boneName=base_jacket_l_middl}
   }
   n86 = "Transform (Modify) Bone - Bone: base_belt"(ComponentPose=@n85, Alpha=@n64, Translation="0,0,0", Rotation="-10,-50,10", Scale="1,1,1") {
    rotationMode=BMM_Additive
    rotationSpace=BCS_BoneSpace
    translationMode=BMM_Additive
    translationSpace=BCS_BoneSpace
    alphaBoolBlend {blendOption=Linear}
    alphaScaleBiasClamp {bInterpResult=true}
    boneToModify {boneName=base_belt}
    rotation=<rot -10 -50 10>
   }
   n87 = Get On Rail()
   n88 = "Transform (Modify) Bone - Bone: base_jacket_r_front"(ComponentPose=@n86, Alpha=@n87, Translation="0,0,0", Rotation="0,-40,10", Scale="1,1,1") {
    rotationMode=BMM_Additive
    rotationSpace=BCS_BoneSpace
    translationMode=BMM_Additive
    translationSpace=BCS_BoneSpace
    alphaBoolBlend {blendOption=Linear}
    alphaScaleBiasClamp {bInterpResult=true}
    boneToModify {boneName=base_jacket_r_front}
    rotation {roll=10, yaw=-40}
   }
   n89 = "Transform (Modify) Bone - Bone: base_belt"(ComponentPose=@n88, Alpha=@n87, Translation="0,0,-2", Rotation="10,20,0", Scale="1,1,1") {
    rotationMode=BMM_Additive
    rotationSpace=BCS_BoneSpace
    translationMode=BMM_Additive
    translationSpace=BCS_BoneSpace
    alphaBoolBlend {blendOption=Linear}
    alphaScaleBiasClamp {bInterpResult=true}
    boneToModify {boneName=base_belt}
    rotation {pitch=10, yaw=20}
    translation {z=-2}
   }
   n90 = "Transform (Modify) Bone - Bone: base_jacket_r_front"(ComponentPose=@n89, Alpha=@n63, Translation="0,0,0", Rotation="0,-20,0", Scale="1,1,1") {
    rotationMode=BMM_Additive
    rotationSpace=BCS_BoneSpace
    translationMode=BMM_Additive
    translationSpace=BCS_BoneSpace
    alphaBoolBlend {blendOption=Linear}
    alphaScaleBiasClamp {bInterpResult=true}
    boneToModify {boneName=base_jacket_r_front}
    rotation {yaw=-20}
   }
   n91 = "Transform (Modify) Bone - Bone: base_jacket_l_front"(ComponentPose=@n90, Alpha=@n63, Translation="0,0,0", Rotation="0,-18,0", Scale="1,1,1") {
    rotationMode=BMM_Additive
    rotationSpace=BCS_BoneSpace
    translationMode=BMM_Additive
    translationSpace=BCS_BoneSpace
    alphaBoolBlend {blendOption=Linear}
    alphaScaleBiasClamp {bInterpResult=true}
    boneToModify {boneName=base_jacket_l_front}
    rotation {yaw=-18}
   }
   n92 = "Transform (Modify) Bone - Bone: base_belt"(ComponentPose=@n91, Alpha=@n63, Translation="0,0,0", Rotation="30,10,0", Scale="1,1,1") {
    rotationMode=BMM_Additive
    rotationSpace=BCS_BoneSpace
    translationMode=BMM_Additive
    translationSpace=BCS_BoneSpace
    alphaBoolBlend {blendOption=Linear}
    alphaScaleBiasClamp {bInterpResult=true}
    boneToModify {boneName=base_belt}
    rotation {pitch=30, yaw=10}
   }
   n93 = "Transform (Modify) Bone - Bone: base_jacket_r_front"(ComponentPose=@n92, Alpha=@n62, Translation="0,0,0", Rotation="0,-5,0", Scale="1,1,1") {
    rotationMode=BMM_Additive
    rotationSpace=BCS_BoneSpace
    translationMode=BMM_Additive
    translationSpace=BCS_BoneSpace
    alphaBoolBlend {blendOption=Linear}
    alphaScaleBiasClamp {bInterpResult=true}
    boneToModify {boneName=base_jacket_r_front}
    rotation {yaw=-5}
   }
   n94 = "Transform (Modify) Bone - Bone: base_jacket_l_front"(ComponentPose=@n93, Alpha=@n62, Translation="0,0,0", Rotation="0,-55,0", Scale="1,1,1") {
    rotationMode=BMM_Additive
    rotationSpace=BCS_BoneSpace
    translationMode=BMM_Additive
    translationSpace=BCS_BoneSpace
    alphaBoolBlend {blendOption=Linear}
    alphaScaleBiasClamp {bInterpResult=true}
    boneToModify {boneName=base_jacket_l_front}
    rotation {yaw=-55}
   }
   n95 = "Transform (Modify) Bone - Bone: base_belt"(ComponentPose=@n94, Alpha=@n62, Translation="0,0,0", Rotation="5,-30,-10", Scale="1,1,1") {
    rotationMode=BMM_Additive
    rotationSpace=BCS_BoneSpace
    translationMode=BMM_Additive
    translationSpace=BCS_BoneSpace
    alphaBoolBlend {blendOption=Linear}
    alphaScaleBiasClamp {bInterpResult=true}
    boneToModify {boneName=base_belt}
    rotation=<rot 5 -30 -10>
   }
   Save cached pose 'Cache_RB_Disable'(Pose=Component To Local(ComponentPose=@n95)) {cacheName=Cache_RB_Disable}
   Reroute Node(InputPin:Wildcard) -> (OutputPin:Wildcard)
  ]}
 EventGraph {nodes [
   n1 = Event BlueprintUpdateAnimation()
   > n2 = Sequence!(execute) -> (
    then_0
    then_1
    then_2:Exec
    then_3:Exec
    then_4:Exec
    then_5:Exec
    then_6:Exec
    then_7:Exec
    then_8:Exec
    then_9:Exec
    then_10:Exec
    then_11:Exec
    then_12:Exec
    then_13:Exec
    then_14:Exec
    then_15:Exec
    then_16:Exec
    then_17:Exec
   )
   n3 = NOT Boolean(A=Get bIsCrouching())
   n4 = NOT Boolean(A=Get bIsFalling())
   n5 = NOT Boolean(A=Get bIsJumping())
   n6 = NOT Boolean(A=Get InAir_Bool())
   n7 = NOT Boolean(A=Get bIsDBNO())
   n8 = Get bIsMoving2D()
   Set Moving2D(
    execute=@n2.then_16
    Moving2D:double="To Float (Boolean)"(InBool=AND Boolean!(A=@n8, B=@n3, C:bool=@n5, D:bool=@n6, E:bool=@n4, F:bool=@n7) -> (ReturnValue))
   ) -> (Output_Get:double)
   n9 = Cast To FortPlayerAnimInstance(execute=@n2.then_14, Object=GetAnimInstance(self=GetOwnerSkeletalMeshComponent()))
   > n10 = Switch on EFortWeaponCoreAnimation!(execute, Selection:EFortWeaponCoreAnimation Enum=Get WeaponCoreAnim(self=@n9)) -> (
    Melee:Exec
    Pistol:Exec
    Shotgun:Exec
    PaperBlueprint:Exec
    Rifle:Exec
    MeleeOneHand:Exec
    MachinePistol:Exec
    RocketLauncher:Exec
    GrenadeLauncher:Exec
    GoingCommando:Exec
    AssaultRifle:Exec
    TacticalShotgun:Exec
    SniperRifle:Exec
    TrapPlacement:Exec
    ShoulderLauncher:Exec
    AbilityDecoTool:Exec
    Crossbow:Exec
    C4:Exec
    RemoteControl:Exec
    DualWield:Exec
    AR_BullPup:Exec
    AR_ForwardGrip:Exec
    MedPackPaddles:Exec
    SMG_P90:Exec
    AR_DrumGun:Exec
    Consumable_Small:Exec
    Consumable_Large:Exec
    Balloon:Exec
    MountedTurret:Exec
    CreativeTool:Exec
    ExplosiveBow:Exec
    AshtonIndigo:Exec
    AshtonChicago:Exec
    MeleeDualWield:Exec
    Unarmed:Exec
   )
   n11 = Branch(
    execute=[@n10.AssaultRifle, @n10.GoingCommando, @n10.GrenadeLauncher, @n10.Rifle, @n10.Shotgun, @n10.SniperRifle, @n10.TacticalShotgun]
    Condition=Get bIsTargeting()
   )
   > Set LongWeaponEquip(LongWeaponEquip:bool=true) -> (Output_Get:bool)
   Set LongWeaponEquip(execute=@n11.else, LongWeaponEquip=false)
   n12 = Gravity Override State Function DEPRECATED(
    execute=@n2.then_0
    CharacterPartAnimInstance=Self-Reference()
    GameTimeInSeconds=GetGameTimeInSeconds()
    GravityOverrideParams=Get Gravity Override Params()
   )
   > Set GravityOverride(GravityOverride=@n12)
   n13 = Get bIsJumping()
   n14 = Get bIsFalling()
   Set IsJumping(execute=@n2.then_4, IsJumping:double="To Float (Boolean)"(InBool=OR Boolean(A=@n13, B=@n14))) -> (Output_Get:double)
   n15 = Get bIsParachuteOpened()
   n16 = OR Boolean!(A=@n15, B=Get bIsSkydiveDiveMode(), C:bool=Get bIsSkydiveFloating()) -> (ReturnValue)
   Set InAir(execute=@n2.then_3, InAir:double="To Float (Boolean)"(InBool=@n16)) -> (Output_Get:double)
   > Set InAir_Bool(InAir_Bool:bool=@n16) -> (Output_Get:bool)
   n17 = Get Skel Mesh()
   n18 = Set HeadBack(
    execute=@n2.then_10
    HeadBack:double=GetAlphaFromReferenceTransformDelta(SkeletalMeshComponent=@n17, BoneName=spine_05, BaseName=head, InputRange=(X=0,Y=90), TwistAxis="0,0,1")
   ) -> (Output_Get:double)
   n19 = float * float(A=@n18, B=3) {is=title K2Node_PromotableOperator}
   > Set Is Hat Back Right(Is Hat Back Right:bool="To Boolean (Integer)"!(InInt:int=Truncate(A=@n19)) -> (ReturnValue:bool)) -> (Output_Get:bool)
   n20 = NOT Boolean(A=Get bIsCrouching())
   n21 = Get Is Hat Back Right() -> (Is Hat Back Right:bool)
   n22 = Get LongWeaponEquip() -> (LongWeaponEquip:bool)
   n23 = Get bIsTargeting()
   Set AimingUp(execute=@n2.then_5, AimingUp:double="To Float (Boolean)"(InBool=AND Boolean!(A=@n21, B=@n20, C=@n22, D=@n23) -> (ReturnValue))) -> (Output_Get:double)
   Set WPO Globals Function(
    execute=@n2.then_15
    Material=Get Mat Instance Dynamic()
    Overlay Material=Get Overlay Mat Instance Dynamic()
    Skeletal Mesh=Get SkelMesh()
    "Local Noise Transform  Location"=Get Local Noise Transform Location() -> (Local Noise Transform Location:Vector)
    Local Noise Offset=Get Local Noise Offset()
   )
   > Set WPO Regional Function!(
    execute
    Material:Material Instance Dynamic Object Reference=Get Mat Instance Dynamic()
    Overlay Material:Material Instance Dynamic Object Reference=Get Overlay Mat Instance Dynamic()
    Skeletal Mesh:Skeletal Mesh Component Object Reference=Get SkelMesh()
    Velocity Transform:Name=foot_r
    Velocity Interp Speed:double=Get Velocity Interp Speed()
    Collision Transform:Name=calf_r
    Collision Transform Axis:EAxis Enum
    Invert Collision Transform Axis:bool=false
    Delta Time X:double=@n1.DeltaTimeX
    Previous Velocity Transform Location:"Vector (by ref)"=Get R1 Velocity Transform Location()
    Previous Velocity Transform Velocity:"Vector (by ref)"=Get R1 Velocity Transform Velocity()
    Region:int=1
   ) -> (then)
   > Set WPO Regional Function(
    Material=Get Mat Instance Dynamic()
    Overlay Material=Get Overlay Mat Instance Dynamic()
    Skeletal Mesh=Get SkelMesh()
    Velocity Transform=foot_l
    Velocity Interp Speed=Get Velocity Interp Speed()
    Collision Transform=calf_l
    Invert Collision Transform Axis=true
    Delta Time X=@n1.DeltaTimeX
    Previous Velocity Transform Location=Get R2 Velocity Transform Location()
    Previous Velocity Transform Velocity=Get R2 Velocity Transform Velocity()
    Region=2
   )
   > n24 = Branch(Condition=Get bIsSkyDiving())
   > Set WPO Pho Wind Intensity(WPO Pho Wind Intensity=0.8)
   > n25 = Set WPO Pho Turbulence(WPO Pho Turbulence:double=1.5) -> (Output_Get:double)
   SetScalarParameterValue(
    execute=[@n25.then, @n27.then]
    self=[Get Mat Instance Dynamic(), Get Overlay Mat Instance Dynamic()]
    ParameterName=ClothSim_WindStrength
    Value=Get WPO Pho Wind Intensity()
   )
   n26 = Get WPO Pho Turbulence() -> (WPO Pho Turbulence:double)
   > SetScalarParameterValue(self=[Get Mat Instance Dynamic(), Get Overlay Mat Instance Dynamic()], ParameterName=ClothSim_R1_TurbulenceStrength, Value=@n26)
   > SetScalarParameterValue(self=[Get Mat Instance Dynamic(), Get Overlay Mat Instance Dynamic()], ParameterName=ClothSim_R2_TurbulenceStrength, Value=@n26)
   Set WPO Pho Wind Intensity(execute=@n24.else, WPO Pho Wind Intensity=0)
   > n27 = Set WPO Pho Turbulence(WPO Pho Turbulence=0.2)
   n28 = NOT Boolean(A=Get bIsSliding())
   n29 = Get bIsCrouching()
   Set InCrouching(execute=@n2.then_1, InCrouching:double="To Float (Boolean)"(InBool=AND Boolean!(A=@n29, B=@n28) -> (ReturnValue))) -> (Output_Get:double)
   Set Is_Wearing_Backpack_or_Cape(
    execute=@n2.then_7
    Is_Wearing_Backpack_or_Cape:bool=OR Boolean!(A=Get isWearingBackbling(), B=Get Is Wearing Cape() -> (Is Wearing Cape:bool)) -> (ReturnValue)
   ) -> (Output_Get:bool)
   Set isParachuting(execute=@n2.then_2, isParachuting:double="To Float (Boolean)"(InBool=@n15)) -> (Output_Get:double)
   n30 = float * float(
    A=GetAlphaFromReferenceTransformDelta(SkeletalMeshComponent=@n17, BoneName=spine_05, BaseName=head, InputRange=(X=0,Y=-90), TwistAxis="0,0,1")
    B=20
   )
   n31 = Set Head Front Bool(execute=@n2.then_11, Head Front Bool:bool="To Boolean (Integer)"(InInt=Truncate(A=@n30))) -> (Output_Get:bool)
   n32 = Get bIsCrouching()
   n33 = Get bIsTargeting()
   > Set Head Front(Head Front:double="To Float (Boolean)"(InBool=AND Boolean!(A=@n31, B=@n32, C:bool=@n33) -> (ReturnValue))) -> (Output_Get:double)
   Set Sliding(execute=@n2.then_9, Sliding:double="To Float (Boolean)"(InBool=Get bIsSliding())) -> (Output_Get:double)
   n34 = NOT Boolean(A=Get bIsSprinting())
   n35 = NOT Boolean(A=Get bIsCrouching())
   n36 = NOT Boolean(A=Get bIsJumping())
   n37 = NOT Boolean(A=Get bIsFalling())
   n38 = NOT Boolean(A=Get bIsCrouchSprinting())
   n39 = Get Leg_L_Rotate_Bool() -> (Leg_L_Rotate_Bool:bool)
   Set On Rail(
    execute=@n2.then_8
    On Rail:double="To Float (Boolean)"(InBool=AND Boolean!(A=@n39, B=@n35, C=@n36, D:bool=@n37, E:bool=@n38, F:bool=@n34) -> (ReturnValue))
   ) -> (Output_Get:double)
   n40 = float * float(
    A=GetAlphaFromReferenceTransformDelta(SkeletalMeshComponent=@n17, BoneName=pelvis, BaseName=thigh_l, InputRange=(X=0,Y=90), TwistAxis="0,0,1")
    B=1.15
   )
   n41 = Set Leg_L_Rotate_Bool(execute=@n2.then_13, Leg_L_Rotate_Bool:bool="To Boolean (Integer)"(InInt=Truncate(A=@n40))) -> (Output_Get:bool)
   n42 = NOT Boolean(A=OR Boolean(A=Get bIsJumping(), B=Get bIsFalling()))
   > Set Leg_L_Rotate(Leg_L_Rotate:double="To Float (Boolean)"(InBool=AND Boolean!(A=@n42, B=@n41) -> (ReturnValue))) -> (Output_Get:double)
   Set Is DBNO(execute=@n2.then_6, Is DBNO:double="To Float (Boolean)"(InBool=Get bIsDBNO())) -> (Output_Get:double)
   n43 = float * float(
    A=GetAlphaFromReferenceTransformDelta(SkeletalMeshComponent=@n17, BoneName=pelvis, BaseName=thigh_r, InputRange=(X=0,Y=90), TwistAxis="0,0,1")
    B=1.15
   )
   n44 = Set Leg_R_Rotate_Bool(execute=@n2.then_12, Leg_R_Rotate_Bool:bool="To Boolean (Integer)"(InInt=Truncate(A=@n43))) -> (Output_Get:bool)
   > Set Leg_R_Rotate(Leg_R_Rotate:double="To Float (Boolean)"(InBool=AND Boolean(A=@n44, B=@n42))) -> (Output_Get:double)
   Event BlueprintInitializeAnimation()
   > Set Skel Mesh(Skel Mesh=GetOwnerSkeletalMeshComponent())
   Event OnCharacterCustomizationLoaded()
   > n45 = Set SkelMesh(SkelMesh=GetPartSkeletalMeshComponent())
   > n46 = Get Dynamic Material Instance(SkeletalMeshComponent=@n45)
   > Set Mat Instance Dynamic(Mat Instance Dynamic=@n46)
   > n47 = GetOrCreateDynamicOverlayMaterial(MeshComponent=@n45, MaterialName=None)
   > Set Overlay Mat Instance Dynamic(Overlay Mat Instance Dynamic=@n47)
   > n48 = Cast To PlayerPawn_Athena(Object=TryGetPawnOwner())
   > Set Body_Mesh(Body_Mesh=GetSkeletalMeshForPartType(self=@n48))
   n49 = CheckForBackbling()
   > Set isWearingBackbling(isWearingBackbling=AND Boolean!(A=@n49, B=NOT Boolean(A=@n49.bIsBackblingACape), C:bool=NOT Boolean(A=@n49.bIsBackblingLong)) -> (ReturnValue))
   > Set Is Wearing Cape(Is Wearing Cape=OR Boolean(A=@n49.bIsBackblingACape, B=@n49.bIsBackblingLong))
  ]}
]
variables [
 AimingUp {type=number}
 Body_Mesh {type=SkeletalMeshComponent}
 Gravity Override Params {type=GravityOverrideParamsStruct}
 GravityOverride {type=Vector}
 Head Front {type=number}
 Head Front Bool {type=boolean}
 HeadBack {type=number}
 InAir {type=number}
 InAir_Bool {type=boolean}
 InCrouching {type=number}
 Is DBNO {type=number}
 Is Hat Back Right {type=boolean}
 Is Wearing Cape {type=boolean}
 Is_Jumping_Bool {type=boolean}
 Is_Wearing_Backpack_or_Cape {type=boolean}
 IsJumping {type=number}
 isParachuting {type=number}
 isSkyDiving {type=number}
 isWearingBackbling {type=boolean}
 Leg_L_Rotate {type=number}
 Leg_L_Rotate_Bool {type=boolean}
 Leg_R_Rotate {type=number}
 Leg_R_Rotate_Bool {type=boolean}
 Local Noise Offset {type=Vector}
 Local Noise Transform Location {type=Vector}
 LongWeaponEquip {type=boolean}
 Mat Instance Dynamic {type=MaterialInstanceDynamic}
 Moving2D {type=number}
 MovingInCrouching {type=number}
 On Rail {type=number}
 Overlay Mat Instance Dynamic {type=MaterialInstanceDynamic}
 R1 Velocity Transform Location {type=Vector}
 R1 Velocity Transform Velocity {type=Vector}
 R2 Velocity Transform Location {type=Vector}
 R2 Velocity Transform Velocity {type=Vector}
 Skel Mesh {type=SkeletalMeshComponent}
 SkelMesh {type=SkeletalMeshComponent}
 Sliding {type=number}
 Sprinting {type=number}
 Velocity Interp Speed {type=number}
 WPO Pho Turbulence {type=number}
 WPO Pho Wind Intensity {type=number}
]
